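# M_ExplosionStreak_02 is an unlit translucent particle sprite material, named for a streak in smoke and dust explosion effects.
# It multiplies the RGB of the T_Explosion_Streak_02 texture by the particle color's alpha, so particle alpha scales the streak's intensity.
# The file does not show which material output receives that product.
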Material /Figment_S04_Geode/Effects/Fort_Effects/Gameplay/SmokeDustExplosions/Materials/M_ExplosionStreak_02
graph {
 properties {
  bUsedWithParticleSprites=true
  blendMode=BLEND_Translucent
  pixelDepthOffsetMode=PDOM_Legacy
  shadingModel=MSM_Unlit
  usedShadingModels=Unlit
 }
 nodes [Multiply(A=ParticleColor().A, B=TextureSample() {texture=/Game/Effects/Fort_Effects/Gameplay/SmokeDustExplosions/Textures/T_Explosion_Streak_02}.RGB)]
}
